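; Exotic grenade launcher Rift Bomb projectile that creates a rift portal where it hits.
; On hit it stops rotating, spawns the launcher's BGA_RiftPortal with the instigator, plays impact and spawn Niagara effects, then destroys itself.
; Gas grenade logic remains: curve-table tick and duration updates, a smoke material instance, and timeline fades for smoke, audio and post-process, none invoked here.
; Water entry only sets a HitWater flag.
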(event Custom|Stop_Rotation
  ;[1 1 1] Stop Rotating Movement
  (Components|Movement|SetUpdatedComponent (Variables|Default|GetRotatingMovement) 0))

(event Custom|Update%20PostProcess%20Scale)

(event Custom|Update%20Tick%20Rate
  (bind _outxy (DataTable|EvaluateCurveTableRow "/Game/Athena/Balance/DataTables/AthenaGameData.AthenaGameData" "Default.GasGrenade.DamagePeriod")
    (:RowFound
      (Variables|GasGrenadeTuning|SetDoT_TickRate _outxy))))

(event Custom|Update%20Gas%20Duration
  (bind _outxy (DataTable|EvaluateCurveTableRow "/Game/Athena/Balance/DataTables/AthenaGameData.AthenaGameData" "Default.GasGrenade.GasDuration")
    (:RowFound
      (Variables|GasGrenadeTuning|SetGasDuration _outxy)
      (Variables|Projectile|SetMaxLifespanOnStop _outxy))))

(event OnEnterWaterEvent(WaterInteractionComponent)
  (params (WaterBody :type Fort%20Water%20Body%20Actor%20Object%20Reference) (WaterInteractionComponent :type Fort%20Water%20Interaction%20Component%20Object%20Reference) (bIsFirstBody :type Boolean))
  (Variables|Default|SetHitWater true))

(event Collision|EventHit
  (params (MyComp :type Primitive%20Component%20Object%20Reference) (Other :type Actor%20Object%20Reference) (OtherComp :type Primitive%20Component%20Object%20Reference) (bSelfMoved :type Boolean) (HitLocation :type Vector :default "0, 0, 0") (HitNormal :type Vector :default "0, 0, 0") (NormalImpulse :type Vector :default "0, 0, 0") (Hit :type Hit%20Result%20Structure%20%28by%20ref%29))
  (Variables|Default|SetLocation HitLocation)
  (Variables|Default|SetTransform (Math|Conversions|ToTransform(Vector) HitLocation))
  (Niagara|SetNiagaraVariable(Float) (Variables|Default|GetNS_Projectile_Globe_Rift_Idle) "Scale_Size" 0.0)
  (CallFunction|StopRotation)
  ;[0.7116 0.838 1] OnStop - Stop the rotation and explode.
  (Building|SpawnBuildingGameplayActor "/HopscotchWeaponsGameplay/Items/GrenadeLauncher/Exotic/BGA_RiftPortal_Athena_Launcher.BGA_RiftPortal_Athena_Launcher_C" (Variables|Default|GetTransform) (Variables|Actor|GetInstigator))
  (Niagara|SpawnSystematLocation (Variables|Default|GetFX_Glass_Rift_Impact) (+ (Variables|Default|GetLocation) "0.000000,0.000000,100.000000"))
  (bind _returnvalue (Niagara|SpawnSystematLocation (Variables|Default|GetFX_Spawn_Rift) (Variables|Default|GetLocation)))
  (Niagara|SetNiagaraVariable(Bool) _returnvalue "blsSpawnIn" true)
  (Actor|DestroyActor))

(event EventOnStop
  (params (Hit :type Hit%20Result%20Structure%20%28by%20ref%29)))

(event Custom|FadeIn
  ;[1 1 1] Timeline
  (Components|Timeline|SetPlayRate (Variables|B_Prj_Athena_RiftBomb|GetFadeSmokeTL) (Utilities|Reciprocal (Variables|Default|GetGasFadeInTime)))
  (Rendering|SetVisibility (Variables|Default|GetSmokeVolumeMesh) true false)
  (bind (_direction _float) (Timeline|FadeSmokeTL PlayFromStart)
    (:Update
      (Rendering|Material|SetScalarParameterValue (Variables|Default|GetSmokeVolumeMeshMID) "Fade" _float :!class /Script/Engine.MaterialInstanceDynamic)
      (Audio|Components|Audio|SetVolumeMultiplier (Variables|Default|GetAudioReleaseSmoke) _float :!class /Script/Engine.AudioComponent)
      (Variables|/Script/Engine.PostProcessComponent|SetBlendWeight (* _float 0.7) (Variables|Default|GetPostProcessComponent)))
    (:Finished
      (Development|PrintString "this faded"))))

(event Custom|FadeOut
  ;[1 1 1] Timeline
  (Components|Timeline|SetPlayRate (Variables|B_Prj_Athena_RiftBomb|GetFadeSmokeTL) (Utilities|Reciprocal (Variables|Default|GetGasFadeOutTime)))
  (Components|Activation|Deactivate (Variables|Default|GetP_GasGrenade_InnerSmoke))
  (Timeline|FadeSmokeTL ReverseFromEnd))

(fn ConstructionScript
  (Networking|CanExecuteCosmeticEvents
    (:True
      (Components|Movement|SetUpdatedComponent (Variables|Default|GetRotatingMovement) 0)
      (Components|Capsule|SetCapsuleSize (Variables|Default|GetOverlapCapsule) (Variables|/Script/Engine.CapsuleComponent|GetCapsuleRadius (Variables|Collision|GetCapsuleComponent)) (Variables|/Script/Engine.CapsuleComponent|GetCapsuleHalfHeight (Variables|Collision|GetCapsuleComponent)) true)
      (bind _returnvalue (Rendering|Material|CreateDynamicMaterialInstance (Variables|Default|GetSmokeVolumeMesh) 0 "/Game/Athena/Items/Consumables/GasGrenade/M_Athena_GasGrenade_Outer_01.M_Athena_GasGrenade_Outer_01" "None" :!class /Script/Engine.PrimitiveComponent))
      (bind _output_get (Variables|Default|SetSmokeVolumeMeshMID _returnvalue))
      (Rendering|Material|SetScalarParameterValue _output_get "Radius" (Variables|Default|GetGasGrenadeRadius) :!class /Script/Engine.MaterialInstanceDynamic)
      (Variables|/Script/Engine.PostProcessComponent|SetBlendWeight 0.0 (Variables|Default|GetPostProcessComponent)))))
